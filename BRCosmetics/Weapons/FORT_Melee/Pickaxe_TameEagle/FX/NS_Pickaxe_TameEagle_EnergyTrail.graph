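# Niagara system for the TameEagle pickaxe: a wobbling energy flare with a ribbon trail that flies between weapon socket, character-relative swing points and impact location.
# User flags for first time, returning, impact found and swing side steer MainFlare through blended targets plus waveform offsets.
# FairyTrail reads MainFlare particles to spawn ribbons, while AppearFlare adds brief flashes at socket or impact.
# System fade and distance checks dim or kill particles; front-end modifiers adjust scale and motion.
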NiagaraSystem /BRCosmetics/Weapons/FORT_Melee/Pickaxe_TameEagle/FX/NS_Pickaxe_TameEagle_EnergyTrail
dependencies {
 usedDataInterfaces [
  &NiagaraDataInterfaceCurve
  &NiagaraDataInterfaceParticleRead
  &NiagaraDataInterfaceColorCurve
  &NiagaraDataInterfaceSocketReader
  &NiagaraDataInterfaceCamera
 ]
 usedDynamicInputs [
  &DistanceBetweenPositions
  &ReturnCameraPosition
  &SetBoolByFloatComparison
  &ClampFloat
 ]
 usedModules [
  &self:SystemSpawnScript.SystemScriptSource.SystemScriptGraph.NiagaraNodeAssignment_1.SetVariables_#1
  &self:SystemSpawnScript.SystemScriptSource.SystemScriptGraph.NiagaraNodeAssignment_2.SetVariables_#2
  &SystemState
  &self:GetSocketPos
  &self:SystemSpawnScript.SystemScriptSource.SystemScriptGraph.NiagaraNodeAssignment_5.SetVariables_#3
  &self:SystemSpawnScript.SystemScriptSource.SystemScriptGraph.NiagaraNodeAssignment_4.SetVariables_#4
  &self:SystemSpawnScript.SystemScriptSource.SystemScriptGraph.NiagaraNodeAssignment_0.SetVariables_#5
 ]
}
emitters {
 AppearFlare {
  emitterUpdateScript [
   &EmitterState(Life Cycle Mode=::NewEnumerator1, Loop Behavior=::NewEnumerator1, Loop Duration=1, UseLoopDelay=false)
   &SpawnBurst_Instantaneous(
    Spawn Time=&MakeCustomFloatFromBool(Bool=$User.FirstTime, True Float=0.03, False Float=0)
    Use Spawn Probability=false
    UseLoopCountLimit=false
   )
   &SpawnBurst_Instantaneous(
    Spawn Group=1
    Spawn Time=&MakeCustomFloatFromBool(Bool=$User.FirstTime, True Float=0.03, False Float=0)
    Use Spawn Probability=false
    UseLoopCountLimit=false
   ) {module=SpawnBurst_Instantaneous001}
  ]
  particleSpawnScript [
   /SetVariables(Particles.VisibilityTag=$Emitter.SpawnGroup) {isSetParameters=true}
   &InitializeParticle(
    Color=<rgba 3 2.6575 2 1>
    Lifetime=&MergeHelper_FloatSwitch(0=0.2, 1=0.3, 2=?, 3=?, Index=$Particles.VisibilityTag, Num=::NewEnumerator0)
    Mesh Scale Mode=::NewEnumerator3
    Mesh Uniform Scale=2
    Sprite Rotation Mode=::NewEnumerator0
    Sprite Size Mode=::NewEnumerator3
    Uniform Sprite Size=30
    UsePositionOffset=true
    Write Material Random=true
   )
   &CameraOffset(Camera Offset Amount=&MergeHelper_FloatSwitch(0=20, 1=5, 2=?, 3=?, Index=$Particles.VisibilityTag, Num=::NewEnumerator0))
  ]
  particleUpdateScript [
   &ParticleState()
   &ScaleSpriteSize(
    Scale Sprite Size Mode=::NewEnumerator0
    Uniform Scale Factor=&FloatFromCurve(
     FloatCurve=DataInterface() {propertyValues {Curve {keys [
         {arriveTangent=5.3732, interpMode=RCIM_Cubic, leaveTangent=5.4155, tangentMode=RCTM_Break, time=0, value=0}
         {arriveTangent=1.6864, interpMode=RCIM_Cubic, leaveTangent=-4.9572, tangentMode=RCTM_Break, time=0.3, value=1}
         {interpMode=RCIM_Cubic, tangentMode=RCTM_Break, time=1, value=0}
        ]}}}
     Scale Curve=$System.SystemFade
    )
   )
   &SocketLocation(
    Interpolate Socket Sampling=true
    Orientation Sampling=::NewEnumerator1
    Position Sampling=::NewEnumerator1
    SamplePosOffset=false
    Socket Index=1
    Socket Reader=$System.Socket Reader
    Socket Sampling Mode=::NewEnumerator3
    VelocitySpeedLimit=false
   )
   &ScaleColor(Scale RGB=&MakeVectorFromLinearColorRGB(Color=$System.TOD), ScaleA=true, ScaleRGB=true) {module=ScaleColor001}
   &ScaleMeshSize(Scale Factor=&VectorFromFloat(Value=&FloatFromCurve(
      FloatCurve=DataInterface() {propertyValues {Curve {defaultValue=1, keys [{tangentMode=RCTM_User, time=0, value=1}, {tangentMode=RCTM_User, time=0.2, value=0.5}]}}}
      Scale Curve=$System.SystemFade
     )))
   /SetVariables() {isSetParameters=true}
   /SetVariables(
    Particles.GlowColor=&MakeLinearColorFromVectorAndFloat(
     "Vector (RGB)"=&Multiply_Vector(
      A=&MakeVectorFromLinearColorRGB(Color=<rgba 1.2 0.6795 0 1>)
      B=&MakeVectorFromLinearColorRGB(Color=$System.TOD)
     )
     "Float (Alpha)"=&Multiply_Float(A=$System.SystemFade, B=0.5)
    )
    Particles.GlowSize=&Multiply_Vector2D_ByFloat(Float=5, Vector2D=$Particles.SpriteSize)
   ) {isSetParameters=true}
   /SetVariables(Particles.Position=&Lerp_Vector(
     A=&MakeCustomVectorFromBool(
      Bool=&Boolean_AndOperation(Bool1=$User.FirstTime, Bool2=&Boolean_NotOperation(Bool=$User.Returning))
      True Vector=Linked() {linkedVariable {name=Output.SocketLocation.SampledPosition, type=NiagaraPosition}}
      False Vector=Linked() {linkedVariable {name=User.ImpactLocation, type=NiagaraPosition}}
     )
     B=Linked() {linkedVariable {name=Particles.Position, type=NiagaraPosition}}
     Alpha=&FloatFromCurve(FloatCurve=DataInterface() {propertyValues {Curve {keys [<at 0.1 0>,<at 0.12 1>]}}}, CurveIndex=$Emitter.Age)
    )) {isSetParameters=true}
   &KillParticles(Kill Particles=$System.DistanceCheck)
   /SetVariables(
    Particles.EnergyColor:LinearColor=&MultiplyLinearColor(
     Linear Color=&MakeLinearColorFromVectorAndFloat(
      "Vector (RGB)"=&MakeVectorFromLinearColorRGB(Color=&ColorFromCurve(ColorCurve=DataInterface() {propertyValues {
          AlphaCurve {keys [<at 0 1 RCIM_Cubic>]}
          BlueCurve {keys [<at 0.0069 0.5331>, {interpMode=RCIM_Cubic, leaveTangent=-0.4609, time=0.132, value=0}, <at 0.7621 0.0758>]}
          GreenCurve {keys [<at 0.0069 0.8698>, {interpMode=RCIM_Cubic, leaveTangent=-0.7294, time=0.132, value=0.8348}, <at 0.7621 0.146>]}
          RedCurve {keys [<at 0.0069 1>, {interpMode=RCIM_Cubic, leaveTangent=-0.7506, time=0.132, value=1}, <at 0.7621 0.2552>]}
         }}))
      "Float (Alpha)"=&FloatFromCurve(FloatCurve=DataInterface() {propertyValues {Curve {keys [<at 0 1 RCIM_Cubic>,<at 0.8 0 RCIM_Cubic>]}}}, Scale Curve=0.1)
     )
     Scale Factor=&MakeVector4_FromLinearColor(LinearColor=$System.TOD)
    )
    Particles.EnergyMeshDyn:Vector4f=&MakeVector4(
     X=&FloatFromCurve(FloatCurve=DataInterface() {propertyValues {Curve {keys [{arriveTangent=2.9193, interpMode=RCIM_Cubic, time=0, value=-1}, {interpMode=RCIM_Cubic, leaveTangent=2, time=1, value=0.5}]}}})
     Y=-1
     Z=0
     W=0
    )
   ) {isSetParameters=true}
  ]
  rendererClasses [&NiagaraSpriteRendererProperties, &NiagaraMeshRendererProperties]
  renderers [
   {rendererClass=&NiagaraSpriteRendererProperties}
   {rendererClass=&NiagaraSpriteRendererProperties, rendererIndex=1}
   {rendererClass=&NiagaraSpriteRendererProperties, rendererIndex=2}
   {rendererClass=&NiagaraMeshRendererProperties, rendererIndex=3}
  ]
 }
 FairyTrail {
  emitterUpdateScript [
   &EmitterState(Life Cycle Mode=::NewEnumerator1, Loop Behavior=::NewEnumerator1, Loop Duration=1, UseLoopDelay=false)
   &SpawnParticlesFromOtherEmitter(
    Attribute Reader=DataInterface() {propertyValues {EmitterBinding {bindingMode=Other, emitterName=MainFlare}}}
    Calculate Spawn Rate Per Particle=false
    Spawn Rate=50
    Use Maximum Age=false
    Use Minimum Age=false
   )
  ]
  particleSpawnScript [
   &InitializeParticle(
    Lifetime=0.25
    Mass Mode=::NewEnumerator3
    Ribbon Width=&Multiply_Float(A=5, B=&FN_FrontEndFloatModifier(Gameplay=1, Front End=$User.FrontEndScale))
    Ribbon Width Mode=::NewEnumerator1
    Sprite Rotation Mode=::NewEnumerator2
    Sprite Size Mode=::NewEnumerator1
    Uniform Sprite Size Max=12
    Uniform Sprite Size Min=6
    UsePositionOffset=false
    Write Material Random=true
   )
   &SampleParticlesFromOtherEmitter(Particle ID Sampling=::NewEnumerator2)
   /SetVariables(Particles.RibbonID=$Output.SampleParticlesFromOtherEmitter.SampledID) {isSetParameters=true}
  ]
  particleUpdateScript [
   &ParticleState()
   &SolveForcesAndVelocity(Clamp Velocity=false, Limit Acceleration=false)
   &ScaleRibbonWidth(Ribbon Width Scale=&Multiply_Float(
     A=&FloatFromCurve(
      FloatCurve=DataInterface() {propertyValues {Curve {keys [
          <at 0 0 RCIM_Cubic>
          {arriveTangent=-11.4315, interpMode=RCIM_Cubic, leaveTangent=-11.4315, tangentMode=RCTM_User, time=0.1, value=4}
          {arriveTangent=-2.8352, interpMode=RCIM_Cubic, leaveTangent=-2.8352, tangentMode=RCTM_User, time=1, value=0}
         ]}}}
      CurveIndex=&Subtract_Float(A=1, B=&ReturnNormalizedExecIndex())
      Scale Curve=&FloatFromCurve(FloatCurve=DataInterface() {propertyValues {Curve {keys [<at 0.25 1 RCIM_Cubic>,<at 0.4 0 RCIM_Cubic>]}}}, CurveIndex=$Emitter.Age)
     )
     B=$System.SystemFade
    ))
   &ScaleColor(
    Scale Alpha=&Multiply_Float(
     A=&FloatFromCurve(
      FloatCurve=DataInterface() {propertyValues {Curve {keys [<at 0.5 1>,<at 0.8 0>]}}}
      CurveIndex=&Subtract_Float(A=1, B=&ReturnNormalizedExecIndex())
      Scale Curve=&FloatFromCurve(FloatCurve=DataInterface() {propertyValues {Curve {keys [<at 0.3 1 RCIM_Cubic>,<at 0.4 0 RCIM_Cubic>]}}}, CurveIndex=$Emitter.Age)
     )
     B=&Multiply_Float(A=$System.SystemFade, B=&FN_FrontEndFloatModifier(Gameplay=1, Front End=0.5))
    )
    Scale RGB=&MakeVectorFromLinearColorRGB(Color=&ColorFromCurve(
      ColorCurve=DataInterface() {propertyValues {
        AlphaCurve {keys [<at 0 1 RCIM_Cubic>,<at 0.4938 1>]}
        BlueCurve {keys [<at 0 0 RCIM_Cubic>,<at 0.5333 0>]}
        GreenCurve {keys [<at 0 6.7048 RCIM_Cubic>,<at 0.5333 0.461>]}
        RedCurve {keys [<at 0 10 RCIM_Cubic>,<at 0.5333 1>]}
       }}
      CurveIndex=&Subtract_Float(A=1, B=&ReturnNormalizedExecIndex())
     ))
    ScaleA=true
    ScaleRGB=true
   )
   &KillParticles(Kill Particles=$System.DistanceCheck)
   /SetVariables(Particles.RibbonWidth_2:NiagaraFloat=&Multiply_Float(A=$Particles.RibbonWidth, B=2)) {isSetParameters=true}
  ]
  rendererClasses [&NiagaraRibbonRendererProperties]
  renderers [
   {rendererClass=&NiagaraRibbonRendererProperties}
   {rendererClass=&NiagaraRibbonRendererProperties, rendererIndex=1}
  ]
 }
 MainFlare {
  emitterUpdateScript [
   &EmitterState(Life Cycle Mode=::NewEnumerator1, Loop Behavior=::NewEnumerator1, Loop Duration=1, UseLoopDelay=false)
   &SpawnBurst_Instantaneous(Use Spawn Probability=false, UseLoopCountLimit=false)
   &SpawnBurst_Instantaneous(
    Spawn Count=&MakeCustomIntFromBool(Bool=$User.Returning, True INT=0, False INT=1)
    Spawn Group=1
    Spawn Time=0.3
    Use Spawn Probability=false
    UseLoopCountLimit=false
   ) {module=SpawnBurst_Instantaneous001}
  ]
  particleSpawnScript [
   /SetVariables(Particles.VisibilityTag=$Emitter.SpawnGroup) {isSetParameters=true}
   &InitializeParticle(
    Color=&MergeHelper_ColorSwitch(0=<rgba 1 0.7127 0.2299 1>, 1=<rgba 0.9896 0.5114 0.0985 0.1>, 2=?, 3=?, Index=$Particles.VisibilityTag, Num=::NewEnumerator0)
    Lifetime=2
    Mass Mode=::NewEnumerator3
    Mesh Scale Mode=::NewEnumerator3
    Mesh Uniform Scale=&Multiply_Float(A=&FN_FrontEndFloatModifier(Gameplay=1, Front End=$User.FrontEndScale), B=0.5)
    Sprite Size Mode=::NewEnumerator3
    Uniform Sprite Size=&Multiply_Float(A=30, B=&FN_FrontEndFloatModifier(Gameplay=1, Front End=$User.FrontEndScale))
    UsePositionOffset=false
    Write Material Random=true
   )
   &SocketLocation(
    Interpolate Socket Sampling=true
    Orientation Sampling=::NewEnumerator1
    SamplePosOffset=false
    Socket Index=1
    Socket Reader=$System.Socket Reader
    Socket Sampling Mode=::NewEnumerator3
    VelocitySpeedLimit=false
   )
   /SetVariables() {isSetParameters=true}
   /SetVariables(Particles.TargetLocation:NiagaraPosition=$Particles.Position) {isSetParameters=true}
   &CameraOffset(Camera Offset Amount=10)
  ]
  particleUpdateScript [
   &ParticleState()
   &ScaleSpriteSize(
    Scale Sprite Size Mode=::NewEnumerator0
    Uniform Scale Factor=&Multiply_Float(
     A=&Multiply_Float(
      A=&MakeCustomFloatFromBool(
       Bool=$User.FirstTime
       True Float=&FloatFromCurve(FloatCurve=DataInterface() {propertyValues {Curve {keys [<at 0 0>,<at 0.1 1>,<at 0.9 1>]}}})
       False Float=&MakeCustomFloatFromBool(
        Bool=$User.Returning
        True Float=&FloatFromCurve(FloatCurve=DataInterface() {propertyValues {Curve {keys [<at 0.3 1>,<at 0.5 0>]}}}, CurveIndex=$Emitter.Age)
        False Float=1
       )
      )
      B=$System.SystemFade
     )
     B=&Waveform(
      Global Phase=&Add_Float(A=$Emitter.Age, B=0.6)
      "Waveform [1]"=::NewEnumerator0
      "[1] Frequency"=1.5
      "[1] Amplitude Min/Max"=<0.8 1>
      "Waveform [2] Blend"=::NewEnumerator2
      "Waveform [2]"=::NewEnumerator9
      "[2] Input Position"=&RandomRangeFloat(Minimum=0, Maximum=10)
     )
    )
   )
   &SolveForcesAndVelocity(Clamp Velocity=false, Limit Acceleration=false)
   &SocketLocation(
    Interpolate Socket Sampling=true
    Orientation Sampling=::NewEnumerator1
    Position Sampling=::NewEnumerator1
    SamplePosOffset=false
    Socket Index=1
    Socket Reader=$System.Socket Reader
    Socket Sampling Mode=::NewEnumerator3
    VelocitySpeedLimit=false
   ) {module=SocketLocation001}
   &SkeletalMeshLocation(
    Bone Sampling Mode=::NewEnumerator4
    "Interpolate Mesh Sampling (Update)"=true
    MaskSpawnGroup=false
    Orientation Sampling=::NewEnumerator1
    Position Sampling=::NewEnumerator2
    Sampled Position Offset=&FN_FrontEndVector3Modifier(
     Gameplay=&MakeCustomVectorFromBool(Bool=$User.RightSwing, True Vector=<-300 0 300>, False Vector=<200 0 300>)
     Front End=<0 0 0>
    )
    Skeletal Mesh=$User.Character
   )
   &ScaleColor(Scale Alpha=$System.SystemFade, ScaleA=true, ScaleRGB=true)
   /SetVariables(Particles.TargetLocation:Vector3f=&Lerp_Vector(
     A=$Particles.TargetLocation
     B=&MakeCustomVectorFromBool(
      Bool=$User.FoundImpact
      True Vector=&TransformVector(Vector=Linked() {linkedVariable {name=User.ImpactLocation, type=NiagaraPosition}}, Source Space=::World, Destination Space=::World)
      False Vector=&Lerp_Vector(
       A=&MakeCustomVectorFromBool(
        Bool=$User.FirstTime
        True Vector=Linked() {linkedVariable {name=Output.SocketLocation001.SampledPosition, type=NiagaraPosition}}
        False Vector=&TransformVector(Vector=Linked() {linkedVariable {name=User.ImpactLocation, type=NiagaraPosition}}, Source Space=::World, Destination Space=::World)
       )
       B=&MakeCustomVectorFromBool(
        Bool=$User.Returning
        True Vector=Linked() {linkedVariable {name=Output.SocketLocation001.SampledPosition, type=NiagaraPosition}}
        False Vector=Linked() {linkedVariable {name=Particles.SkeletalMeshLocation.SampledPosition, type=NiagaraPosition}}
       )
       Alpha=&FloatFromCurve(FloatCurve=DataInterface() {propertyValues {Curve {keys [<at 0.1 0>,<at 0.3 1>]}}}, CurveIndex=$Emitter.Age)
      )
     )
     Alpha=0.7
    )) {isSetParameters=true}
   /SetVariables(Particles.Position=&AddVectorToPosition(
     Position=Linked() {linkedVariable {name=Particles.TargetLocation, type=Vector3f}}
     Vector=&Multiply_VectorByFloat(
      Vector=&FN_FrontEndVector3Modifier(
       Gameplay=&MakeVector(
        X=&Waveform(
         Global Input Position=&RandomRangeFloat(Minimum=0, Maximum=100)
         "Waveform [1]"=::NewEnumerator1
         "[1] Frequency"=&RandomRangeFloat(Minimum=2, Maximum=3)
         "[1] Amplitude Min/Max"=<-1 1>
         "[1] Scale"=&RandomRangeFloat(Minimum=30, Maximum=60)
        )
        Y=&Waveform(
         Global Input Position=&RandomRangeFloat(Minimum=0, Maximum=100)
         "Waveform [1]"=::NewEnumerator0
         "[1] Frequency"=&RandomRangeFloat(Minimum=2, Maximum=3)
         "[1] Amplitude Min/Max"=<-1 1>
         "[1] Scale"=&RandomRangeFloat(Minimum=30, Maximum=60)
        )
        Z=&Waveform(
         Global Input Position=&RandomRangeFloat(Minimum=0, Maximum=100)
         "Waveform [1]"=::NewEnumerator10
         "[1] Frequency"=&RandomRangeFloat(Minimum=2, Maximum=4)
         "[1] Amplitude Min/Max"=<-1 1>
         "[1] Scale"=&RandomRangeFloat(Minimum=30, Maximum=60)
        )
       )
       Front End=&MakeCustomVectorFromBool(
        Bool=$User.FirstTime
        True Vector=&MakeVector(
         X=&Waveform(
          Global Input Position=&RandomRangeFloat(Minimum=0, Maximum=100)
          "Waveform [1]"=::NewEnumerator1
          "[1] Frequency"=&RandomRangeFloat(Minimum=2, Maximum=3)
          "[1] Amplitude Min/Max"=<-1 1>
          "[1] Scale"=&RandomRangeFloat(Minimum=30, Maximum=60)
         )
         Y=&Waveform(
          Global Input Position=&RandomRangeFloat(Minimum=0, Maximum=100)
          "Waveform [1]"=::NewEnumerator0
          "[1] Frequency"=&RandomRangeFloat(Minimum=2, Maximum=3)
          "[1] Amplitude Min/Max"=<-1 1>
          "[1] Scale"=&RandomRangeFloat(Minimum=30, Maximum=60)
         )
         Z=&RandomRangeFloat(Minimum=-50, Maximum=-70)
        )
        False Vector=&Add_Vector(
         A=&MakeVector(
          X=&Waveform(
           Global Input Position=&RandomRangeFloat(Minimum=0, Maximum=100)
           "Waveform [1]"=::NewEnumerator1
           "[1] Frequency"=&RandomRangeFloat(Minimum=2, Maximum=3)
           "[1] Amplitude Min/Max"=<-1 1>
           "[1] Scale"=&RandomRangeFloat(Minimum=30, Maximum=60)
          )
          Y=&Waveform(
           Global Input Position=&RandomRangeFloat(Minimum=0, Maximum=100)
           "Waveform [1]"=::NewEnumerator0
           "[1] Frequency"=&RandomRangeFloat(Minimum=2, Maximum=3)
           "[1] Amplitude Min/Max"=<-1 1>
           "[1] Scale"=&RandomRangeFloat(Minimum=30, Maximum=60)
          )
          Z=&RandomRangeFloat(Minimum=-60, Maximum=20)
         )
         B=&Add_Vector(
          A=<0 0 50>
          B=&Multiply_VectorByFloat(
           Vector=&Subtract_Vector(A=Linked() {linkedVariable {name=Particles.SkeletalMeshLocation.SampledPosition, type=NiagaraPosition}}, B=$Particles.TargetLocation)
           Float=&RandomRangeFloat(Minimum=0.2, Maximum=0.5)
          )
         )
        )
       )
      )
      Float=&FloatFromCurve(
       FloatCurve=DataInterface() {propertyValues {Curve {keys [
           {arriveTangent=0.1585, interpMode=RCIM_Cubic, leaveTangent=0.1585, tangentMode=RCTM_User, time=0, value=0}
           <at 0.15 1 RCIM_Cubic>
           <at 0.25 0 RCIM_Cubic>
          ]}}}
       CurveIndex=$Emitter.Age
      )
     )
    )) {isSetParameters=true}
   /SetVariables(
    Particles.GlowColor=&MakeLinearColorFromVectorAndFloat(
     "Vector (RGB)"=&Multiply_Vector(
      A=&MakeVectorFromLinearColorRGB(Color=<rgba 0.3688 0.0919 0 1>)
      B=&MakeVectorFromLinearColorRGB(Color=$System.TOD)
     )
     "Float (Alpha)"=&Multiply_Float(A=0.5, B=$System.SystemFade)
    )
    Particles.GlowSize=&Multiply_Vector2D_ByFloat(Float=3, Vector2D=$Particles.SpriteSize)
   ) {isSetParameters=true}
   /SetVariables(
    Particles.FlareColor=&MakeLinearColorFromVectorAndFloat(
     "Vector (RGB)"=&Multiply_Vector(
      A=&MakeVectorFromLinearColorRGB(Color=<rgba 2 1.8878 1.6594 1>)
      B=&MakeVectorFromLinearColorRGB(Color=$System.TOD)
     )
     "Float (Alpha)"=&Multiply_Float(A=0.5, B=$System.SystemFade)
    )
    Particles.FlareSize:Vector2f=&Multiply_Vector2D_ByFloat(Float=0.5, Vector2D=$Particles.SpriteSize)
   ) {isSetParameters=true}
   /SetVariables(
    Particles.GlowColor_2:LinearColor=&MakeLinearColorFromVectorAndFloat(
     "Vector (RGB)"=&Multiply_Vector(
      A=&MakeVectorFromLinearColorRGB(Color=<rgba 0.625 0.2333 0.0139 1>)
      B=&MakeVectorFromLinearColorRGB(Color=$System.TOD)
     )
     "Float (Alpha)"=&Multiply_Float(A=0.2, B=$System.SystemFade)
    )
    Particles.GlowSize_2:Vector2f=&Multiply_Vector2D_ByFloat(Float=3, Vector2D=$Particles.SpriteSize)
   ) {isSetParameters=true}
   /SetVariables(
    Particles.GlowColor_3:LinearColor=&MakeLinearColorFromVectorAndFloat(
     "Vector (RGB)"=&Multiply_Vector(
      A=&MakeVectorFromLinearColorRGB(Color=<rgba 0.625 0.4382 0.0772 1>)
      B=&MakeVectorFromLinearColorRGB(Color=$System.TOD)
     )
     "Float (Alpha)"=&Multiply_Float(A=&Multiply_Float(A=0.2, B=$System.SystemFade), B=1)
    )
    Particles.GlowSize_3:Vector2f=&Multiply_Vector2D_ByFloat(Float=4, Vector2D=$Particles.SpriteSize)
   ) {isSetParameters=true}
   &KillParticles(Kill Particles=$System.DistanceCheck)
   /SetVariables(Particles.PulseDyn:Vector4f=&MakeVector4(
     X=&RemapRange(
      Input Value To Remap=&FloatFromCurve(
       FloatCurve=DataInterface() {propertyValues {Curve {keys [
           {arriveTangent=-0.4027, interpMode=RCIM_Cubic, leaveTangent=-0.4027, tangentMode=RCTM_User, time=0, value=0}
           {arriveTangent=1.6708, interpMode=RCIM_Cubic, leaveTangent=1.6708, tangentMode=RCTM_User, time=1, value=1}
          ]}}}
       CurveIndex=&Subtract_Float(
        A=1
        B=&Frac_Float(Float=&Add_Float(A=&Multiply_Float(A=$Emitter.Age, B=1.5), B=0.8))
       )
      )
      Input Min=0
      Input Max=1
      Output Min=0
      Output Max=-0.8
     )
     Y=1
     Z=&RemapRange(
      Input Value To Remap=&Subtract_Float(
       A=1
       B=&Frac_Float(Float=&Add_Float(A=&Multiply_Float(A=$Emitter.Age, B=1.5), B=0.8))
      )
      Input Min=0
      Input Max=1
      Output Min=0.5
      Output Max=0.5
     )
     W=0
    )) {isSetParameters=true}
  ]
  rendererClasses [&NiagaraSpriteRendererProperties, &NiagaraMeshRendererProperties]
  renderers [
   {rendererClass=&NiagaraSpriteRendererProperties}
   {rendererClass=&NiagaraSpriteRendererProperties, rendererIndex=1}
   {rendererClass=&NiagaraSpriteRendererProperties, rendererIndex=2}
   {rendererClass=&NiagaraSpriteRendererProperties, rendererIndex=3}
   {rendererClass=&NiagaraSpriteRendererProperties, rendererIndex=4}
   {rendererClass=&NiagaraSpriteRendererProperties, rendererIndex=5}
   {rendererClass=&NiagaraMeshRendererProperties, rendererIndex=6}
  ]
 }
}
summary {
 systemName=NS_Pickaxe_TameEagle_EnergyTrail
 emitters [
  {emitterName=MainFlare, rendererClasses [&NiagaraSpriteRendererProperties, &NiagaraMeshRendererProperties]}
  {emitterName=FairyTrail, rendererClasses [&NiagaraRibbonRendererProperties]}
  {emitterName=AppearFlare, rendererClasses [&NiagaraSpriteRendererProperties, &NiagaraMeshRendererProperties]}
 ]
 userVariables [
  User.bIsFrontEnd {type=NiagaraBool}
  User.bIsFrontEndPreview {type=NiagaraBool}
  User.Character {
   type=NiagaraDataInterfaceSkeletalMesh
   defaultValue {
    struct=NiagaraExt_VariableValue_DataInterface
    value {dataInterface=&self:NiagaraDataInterfaceSkeletalMesh_0}
   }
  }
  User.FirstTime {type=NiagaraBool}
  User.FoundImpact {type=NiagaraBool}
  User.FrontEndScale {type=NiagaraFloat, defaultValue=0.5, description=""}
  User.ImpactLocation {type=NiagaraPosition, description="", defaultValue {}}
  User.Returning {type=NiagaraBool}
  User.RightSwing {type=NiagaraBool}
 ]
}
