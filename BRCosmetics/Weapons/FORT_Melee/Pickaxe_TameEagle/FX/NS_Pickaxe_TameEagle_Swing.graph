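# Niagara system for the swing trail of the TameEagle pickaxe, built from three ribbon emitters.
# Each uses anim-trail modules on bone or socket indices: a main Trail, a six-trail Tip_Trail that narrows over life, and a two-trail ToonTrail.
# User.GlowAmount blends colors from teal-green toward gold, combined with system TOD tint and fade values.
# A system front-end scale sets the width of Trail and ToonTrail.
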NiagaraSystem /BRCosmetics/Weapons/FORT_Melee/Pickaxe_TameEagle/FX/NS_Pickaxe_TameEagle_Swing
dependencies {
 usedDataInterfaces [&NiagaraDataInterfaceColorCurve, &NiagaraDataInterfaceCurve, &NiagaraDataInterfaceSocketReader]
 usedDynamicInputs [
  &FN_FrontEndFloatModifier
  &ClampFloat
  &Add_Float
  &Multiply_Float
  &MakeCustomFloatFromBool
 ]
 usedModules [
  &self:SystemSpawnScript.SystemScriptSource.SystemScriptGraph.NiagaraNodeAssignment_1.SetVariables_#1
  &self:SystemSpawnScript.SystemScriptSource.SystemScriptGraph.NiagaraNodeAssignment_2.SetVariables_#2
  &self:SystemSpawnScript.SystemScriptSource.SystemScriptGraph.NiagaraNodeAssignment_5.SetVariables_#3
  &self:SystemSpawnScript.SystemScriptSource.SystemScriptGraph.NiagaraNodeAssignment_0.SetVariables_#4
  &SystemState
  &self:SystemSpawnScript.SystemScriptSource.SystemScriptGraph.NiagaraNodeAssignment_3.SetVariables_#5
  &self:SystemSpawnScript.SystemScriptSource.SystemScriptGraph.NiagaraNodeAssignment_4.SetVariables_#6
 ]
}
emitters {
 Tip_Trail {
  emitterSpawnScript [&FN_RibbonIDOffsetUtility()]
  emitterUpdateScript [
   &EmitterState()
   &FN_SpawnAnimTrail(Number of Anim Trails=6)
  ]
  particleSpawnScript [
   &FN_InitializeAnimTrailBoneIndecies(Second Bone or Socket Index=0) {module=FN_InitializeAnimTrailBoneIndecies001}
   &FN_InitializeAnimTrailBoneIndecies(AnimTrail Selection=1, First Bone or Socket Index=1) {module=FN_InitializeAnimTrailBoneIndecies005}
   &FN_InitializeAnimTrailBoneIndecies(AnimTrail Selection=2, First Bone or Socket Index=4, Second Bone or Socket Index=4) {module=FN_InitializeAnimTrailBoneIndecies004}
   &FN_InitializeAnimTrailBoneIndecies(AnimTrail Selection=3, First Bone or Socket Index=5, Second Bone or Socket Index=5) {module=FN_InitializeAnimTrailBoneIndecies003}
   &FN_InitializeAnimTrailBoneIndecies(AnimTrail Selection=4, First Bone or Socket Index=6, Second Bone or Socket Index=6) {module=FN_InitializeAnimTrailBoneIndecies002}
   &FN_InitializeAnimTrailBoneIndecies(AnimTrail Selection=5, First Bone or Socket Index=1)
   &FN_InitializeAnimTrail(
    Color=&MultiplyLinearColor(
     Linear Color=<rgba 0.1974 1 0.5891 0.2>
     Scale Factor=&MakeVector4FromVectorAndFloat("Vector (RGB)"=&MakeVectorFromLinearColorRGB(Color=$System.TOD), "Float (Alpha)"=$System.SystemFade)
    )
    Lifetime=0.15
   )
  ]
  particleUpdateScript [
   &FN_UpdateAnimTrail() {module=FN_UpdateAnimTrail001}
   &ScaleRibbonWidth(Ribbon Width Scale=&FloatFromCurve(FloatCurve=DataInterface() {propertyValues {Curve {keys [<at 0 1 RCIM_Cubic>, {arriveTangent=-10.0559, interpMode=RCIM_Cubic, tangentMode=RCTM_Break, time=0.2, value=0}]}}}))
   &Color(Color=&MakeLinearColorFromVectorAndFloat(
     "Vector (RGB)"=&MakeVectorFromLinearColorRGB(Color=&LerpLinearColors(
       StartColor=&MultiplyLinearColor(
        Linear Color=$System.TOD
        Scale Factor=&MakeVector4_FromLinearColor(LinearColor=<rgba 0.1974 1 0.5891 0.2>)
       )
       EndColor=<rgba 1 0.8932 0.1844 0.2>
       LerpFactor=$User.GlowAmount
      ))
     "Float (Alpha)"=1
    ))
  ]
  rendererClasses [&NiagaraRibbonRendererProperties]
  renderers [
   {rendererClass=&NiagaraRibbonRendererProperties}
   {rendererClass=&NiagaraRibbonRendererProperties, rendererIndex=1}
  ]
 }
 ToonTrail {
  emitterSpawnScript [&FN_RibbonIDOffsetUtility()]
  emitterUpdateScript [
   &EmitterState()
   &FN_SpawnAnimTrail(Number of Anim Trails=2)
  ]
  particleSpawnScript [
   &FN_InitializeAnimTrailBoneIndecies()
   &FN_InitializeAnimTrailBoneIndecies(AnimTrail Selection=1, First Bone or Socket Index=2, Second Bone or Socket Index=3) {module=FN_InitializeAnimTrailBoneIndecies001}
   &FN_InitializeAnimTrail(
    AnimTrailWidth=$System.FrontEndScale
    Color=&MultiplyLinearColor(
     Linear Color=$System.TOD
     Scale Factor=&MakeVector4_FromLinearColor(LinearColor=&MergeHelper_ColorSwitch(0=<rgba 0.1526 0.7917 0.5815 1>, 1=<rgba 0.0833 0.0833 0.0833 1>, 2=?, 3=?, Index=$Particles.SpawnGroup, Num=::NewEnumerator0))
    )
    Lifetime=0.2
   )
  ]
  particleUpdateScript [
   &FN_UpdateAnimTrail() {module=FN_UpdateAnimTrail001}
   &FN_Sprite_Attributes(
    Color=&MakeLinearColorFromVectorAndFloat(
     "Vector (RGB)"=&MakeVectorFromLinearColorRGB(Color=&MultiplyLinearColor(
       Linear Color=<rgba 0.326 0.3827 0.4 1>
       Scale Factor=&MakeVector4_FromLinearColor(LinearColor=&ColorFromCurve(ColorCurve=DataInterface() {propertyValues {
           AlphaCurve {keys [<at 0 1 RCIM_Cubic>,<at 0.6667 0>]}
           BlueCurve {keys [<at 0 1 RCIM_Cubic>,<at 0.6667 0.05 RCIM_Cubic>]}
           GreenCurve {keys [<at 0 1 RCIM_Cubic>,<at 0.6667 0.05 RCIM_Cubic>]}
           RedCurve {keys [<at 0 1 RCIM_Cubic>,<at 0.6667 0.05 RCIM_Cubic>]}
          }}))
      ))
     "Float (Alpha)"=1
    )
    Create Color Attribute=true
    Write to Module Namespace=true
   )
   &Color(Color=&MakeLinearColorFromVectorAndFloat(
     "Vector (RGB)"=&MakeVectorFromLinearColorRGB(Color=&LerpLinearColors(
       StartColor=&MultiplyLinearColor(
        Linear Color=$System.TOD
        Scale Factor=&MakeVector4_FromLinearColor(LinearColor=&MergeHelper_ColorSwitch(0=<rgba 0.1526 0.7917 0.5815 1>, 1=<rgba 0.0833 0.0833 0.0833 1>, 2=?, 3=?, Index=$Particles.SpawnGroup, Num=::NewEnumerator0))
       )
       EndColor=<rgba 1 0.8932 0.1844 1>
       LerpFactor=$User.GlowAmount
      ))
     "Float (Alpha)"=$System.SystemFadeToon
    ))
  ]
  rendererClasses [&NiagaraRibbonRendererProperties]
  renderers [{rendererClass=&NiagaraRibbonRendererProperties}]
 }
 Trail {
  emitterUpdateScript [
   &EmitterState()
   &FN_SpawnAnimTrail()
  ]
  particleSpawnScript [
   &FN_InitializeAnimTrailBoneIndecies()
   &FN_InitializeAnimTrail(
    AnimTrailWidth=$System.FrontEndScale
    Color=&MultiplyLinearColor(
     Linear Color=<rgba 1 1 1 0.502>
     Scale Factor=&MakeVector4FromVectorAndFloat("Vector (RGB)"=&MakeVectorFromLinearColorRGB(Color=$System.TOD), "Float (Alpha)"=$System.SystemFade)
    )
    Lifetime=0.2
   )
  ]
  particleUpdateScript [
   &FN_UpdateAnimTrail() {module=FN_UpdateAnimTrail001}
   &ScaleColor(
    Scale Alpha=&FloatFromCurve(FloatCurve=DataInterface() {propertyValues {Curve {keys [<at 0.1 1>,<at 0.15 0>]}}}, CurveIndex=$Particles.Age)
    Scale RGB=&MakeVectorFromLinearColorRGB(Color=&LerpLinearColors(
      StartColor=&ColorFromCurve(ColorCurve=DataInterface() {propertyValues {
         AlphaCurve {keys []}
         BlueCurve {keys [{arriveTangent=-0.8311, interpMode=RCIM_Cubic, time=0.102, value=0.7827}, <at 0.4489 0.728>, <at 0.9834 1>]}
         GreenCurve {keys [{arriveTangent=-1.9648, interpMode=RCIM_Cubic, time=0.102, value=1}, <at 0.4489 0.8578>, <at 0.9834 0.5679>]}
         RedCurve {keys [{arriveTangent=-0.3647, interpMode=RCIM_Cubic, time=0.102, value=0.2604}, <at 0.4489 0.2333>, <at 0.9834 0.1282>]}
        }})
      EndColor=&ColorFromCurve(ColorCurve=DataInterface() {propertyValues {
         AlphaCurve {keys []}
         BlueCurve {keys [{arriveTangent=-0.8311, interpMode=RCIM_Cubic, time=0.102, value=0.2657}, <at 0.4489 0.1705>, <at 0.9834 0.0938>]}
         GreenCurve {keys [{arriveTangent=-1.9648, interpMode=RCIM_Cubic, time=0.102, value=0.8739}, <at 0.4489 0.6509>, <at 0.9834 0.3429>]}
         RedCurve {keys [{arriveTangent=-0.3647, interpMode=RCIM_Cubic, time=0.102, value=1}, <at 0.4489 0.8578>, <at 0.9834 0.6094>]}
        }})
      LerpFactor=$User.GlowAmount
     ))
    ScaleA=true
    ScaleRGB=true
   )
  ]
  rendererClasses [&NiagaraRibbonRendererProperties]
  renderers [{rendererClass=&NiagaraRibbonRendererProperties}]
 }
}
summary {
 systemName=NS_Pickaxe_TameEagle_Swing
 emitters [
  {emitterName=Trail, rendererClasses [&NiagaraRibbonRendererProperties]}
  {emitterName=Tip_Trail, rendererClasses [&NiagaraRibbonRendererProperties]}
  {emitterName=ToonTrail, rendererClasses [&NiagaraRibbonRendererProperties]}
 ]
 userVariables [
  User.bIsFrontEnd {type=NiagaraBool}
  User.bIsFrontEndPreview {type=NiagaraBool}
  User.EnableGlow {type=NiagaraBool}
  User.GlowAmount {type=NiagaraFloat, defaultValue=0, description=""}
  User.RibbonIDOffset {type=NiagaraInt32}
 ]
}
